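# Translucent unlit material for the Heartbeat drift trail on Niagara ribbons and mesh particles, drawing a pulse line over a scrolling grid.
# Two panned mask samples displace a V-based line up and down, and steep bias-scale thresholds cut it into a thin trace.
# IsWheelWrap switches between a U-faded trail and a wheel-masked variant with added edge lines.
# Particle color tints the result through emissive adjustments; the same pattern drives opacity.
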Material /VehicleCosmetics/DriftTrail/DriftTrail_Heartbeat/Materials/HeartBeat_MAT
graph {
 properties {
  bUsedWithNiagaraMeshParticles=true
  bUsedWithNiagaraRibbons=true
  blendMode=BLEND_Translucent
  pixelDepthOffsetMode=PDOM_Legacy
  shadingModel=MSM_Unlit
  usedShadingModels=Unlit
 }
 nodes [
  MaterialFunctionCall(Z=ScalarParameter() {parameterName=WPO}) {materialFunction=&MakeFloat3}
  n1 = Power(Base=ComponentMask(None=TextureCoordinate()) {g=true}) {constExponent=1.4}
  n2 = ConstantBiasScale(None=@n1) {bias=-0.07, scale=1024}
  n3 = OneMinus(None=ConstantBiasScale(None=@n1) {bias=-0.11, scale=1024})
  n4 = TextureCoordinate()
  n5 = Power(Base=OneMinus(None=ComponentMask(None=@n4) {g=true})) {constExponent=1.4}
  n6 = ConstantBiasScale(None=@n5) {bias=-0.07, scale=1024}
  n7 = OneMinus(None=ConstantBiasScale(None=@n5) {bias=-0.11, scale=1024})
  n8 = Add(A=Saturate(None=Multiply(A=@n2, B=@n3)), B=Saturate(None=Multiply(A=@n6, B=@n7)))
  n9 = TextureCoordinate()
  n10 = Multiply(A=@n9, B=Constant2Vector() {g=0.01, r=3})
  n11 = Panner(Coordinate=@n10) {speedX=-3, speedY=1}
  n12 = TextureObject() {texture=/DealershipSharedLibrary/FX/Textures/Gradient_Masks/T_VCVFX_SphereMaskTimeOffset_A}
  n13 = Saturate(None=ConstantBiasScale(None=TextureSample(UVs=@n11, Tex=@n12) {texture=&T_VCVFX_SphereMaskTimeOffset_A}.R) {bias=-0.15, scale=1})
  n14 = TextureCoordinate()
  n15 = ComponentMask(None=@n14) {g=true}
  n16 = Multiply(A=@n15, B=OneMinus(None=@n15))
  n17 = ComponentMask(None=@n14) {r=true}
  n18 = Saturate(None=OneMinus(None=NamedRerouteDeclaration(None=@n17) {name=Mask R, nodeColor=<rgba 0.5469 0 1 1>}))
  n19 = Multiply(A=Saturate(None=Multiply(A=@n16) {constB=3}), B=@n18)
  n20 = ScalarParameter() {defaultValue=0.3, parameterName=Offset MainLine}
  n21 = Add(A=ComponentMask(None=Add(A=@n9, B=@n20)) {g=true}) {constB=-1.25}
  n22 = Panner(Coordinate=@n10) {speedX=-1.5, speedY=0.7}
  n23 = Saturate(None=ConstantBiasScale(None=TextureSample(UVs=@n22, Tex=@n12) {texture=&T_VCVFX_SphereMaskTimeOffset_A}.R) {bias=-0.15, scale=1.5})
  n24 = Multiply(A=Subtract(A=@n21, B=Multiply(A=@n13, B=@n19)), B=Add(A=@n21, B=Multiply(A=@n23, B=@n19)))
  n25 = ConstantBiasScale(None=@n24) {bias=-0.1, scale=1024}
  n26 = ConstantBiasScale(None=@n24) {bias=-0.12, scale=1024}
  n27 = Saturate(None=Multiply(A=Saturate(None=@n25), B=OneMinus(None=@n26)))
  n28 = Add(A=Multiply(A=@n27, B=ScalarParameter() {defaultValue=1, parameterName=Line}), B=Multiply(A=@n27) {constB=50})
  n29 = ComponentMask(None=TextureCoordinate()) {r=true}
  n30 = OneMinus(None=@n29)
  n31 = Power(Base=@n29) {constExponent=1.2}
  n32 = Multiply(A=Multiply(A=@n30, B=@n30), B=Saturate(None=Multiply(A=@n31) {constB=20}))
  n33 = Multiply(A=Multiply(A=Saturate(None=@n32), B=@n27)) {constB=5}
  n34 = TextureCoordinate()
  n35 = AppendVector(
   A=ScalarParameter() {defaultValue=2, group=Secondary Texture, parameterName=Tiling X}
   B=ScalarParameter() {defaultValue=1, group=Secondary Texture, parameterName=Tiling Y}
  )
  n36 = ScalarParameter() {group=Secondary Texture, parameterName=Speed X}
  n37 = Constant()
  n38 = TextureSampleParameter2D(UVs=Panner(Coordinate=Multiply(A=@n34, B=@n35), Speed=AppendVector(A=@n36, B=@n37))) {
   group=Secondary Texture
   parameterName=Secondary Texture
   texture=/DealershipSharedLibrary/FX/Textures/Gradient_Masks/T_VCVFX_Trail_Heartbeat_Grid_A
  }
  n39 = Multiply(A=@n38.RGB, B=@n38.RGB)
  n40 = MaterialFunctionCall!(Multiply, Power EXP=ScalarParameter() {defaultValue=0.96, parameterName=Opacity Wheel}) {materialFunction=/DealershipSharedLibrary/FX/Materials/MaterialFunctions/MF_VCVFX_WheelwrapMask}
  n41 = Multiply(A=Multiply(A=Multiply(A=@n39) {constB=3}, B=ScalarParameter() {defaultValue=1, parameterName=Grid}), B=@n40.Result)
  n42 = Multiply(A=@n39, B=Multiply(A=@n18, B=Multiply(A=@n16) {constB=5}))
  n43 = Add(
   A=StaticSwitchParameter(True=Multiply(A=@n8) {constB=3}, False=Constant()) {parameterName=IsWheelWrap?}
   B=Add(
    A=StaticSwitchParameter(True=@n28, False=@n33) {parameterName=IsWheelWrap?}
    B=StaticSwitchParameter(True=@n41, False=@n42) {parameterName=IsWheelWrap?}
   )
  )
  MaterialFunctionCall(Emissive=Multiply(A=ParticleColor().RGB, B=@n43)) {materialFunction=&MF_FX_EmissiveAdjustments}
  n44 = NamedRerouteUsage()
  StaticSwitchParameter(True=Multiply(A=Saturate(None=@n43), B=@n40.Result), False=Saturate(None=Multiply(A=@n43, B=@n44))) {parameterName=IsWheelWrap?}
 ]
}
