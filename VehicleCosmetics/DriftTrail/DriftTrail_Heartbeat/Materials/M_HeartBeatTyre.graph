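# Unlit translucent material for Niagara mesh particles in the heartbeat drift trail tyre effect, drawing a glowing pulse line over a scrolling grid.
# Scale-bias thresholds cut thin edge bands and a central line, which two panning sphere-mask textures displace into a moving waveform.
# A panned thread texture adds the grid, with Line and Grid scalars weighting the layers.
# Particle colour tints the emissive sum, and a rounded sine of V masks its saturated alpha.
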Material /VehicleCosmetics/DriftTrail/DriftTrail_Heartbeat/Materials/M_HeartBeatTyre
graph {
 properties {
  bUsedWithNiagaraMeshParticles=true
  blendMode=BLEND_Translucent
  pixelDepthOffsetMode=PDOM_Legacy
  shadingModel=MSM_Unlit
  usedShadingModels=Unlit
 }
 nodes [
  n1 = Power(Base=ComponentMask(None=TextureCoordinate()) {g=true}) {constExponent=1.2}
  n2 = ConstantBiasScale(None=@n1) {bias=-0.12, scale=1024}
  n3 = Saturate(None=Multiply(A=ConstantBiasScale(None=@n1) {bias=-0.1, scale=1024}, B=OneMinus(None=@n2)))
  n4 = TextureCoordinate()
  n5 = Power(Base=OneMinus(None=ComponentMask(None=@n4) {g=true})) {constExponent=1.3}
  n6 = ConstantBiasScale(None=@n5) {bias=-0.12, scale=1024}
  n7 = Saturate(None=Multiply(A=ConstantBiasScale(None=@n5) {bias=-0.1, scale=1024}, B=OneMinus(None=@n6)))
  n8 = TextureCoordinate()
  n9 = Add(A=@n8, B=Constant2Vector() {g=0.12})
  n10 = Constant() {r=1}
  n11 = Add(A=ComponentMask(None=Panner(Coordinate=@n9, Time=@n10) {speedY=0.3}) {g=true}) {constB=-1.25}
  n12 = Multiply(A=Constant2Vector() {g=0.01, r=3}, B=@n8)
  n13 = Panner(Coordinate=@n12) {speedX=-3, speedY=1}
  n14 = Saturate(None=ConstantBiasScale(None=TextureSample(UVs=@n13) {texture=/DealershipSharedLibrary/FX/Textures/Gradient_Masks/T_VCVFX_SphereMaskTimeOffset_A}.R) {bias=-0.15, scale=1})
  n15 = TextureCoordinate()
  n16 = ComponentMask(None=@n15) {g=true}
  n17 = Multiply(A=@n16, B=OneMinus(None=@n16))
  n18 = ComponentMask(None=@n15) {r=true}
  n19 = Multiply(A=Saturate(None=Multiply(A=@n17) {constB=3}), B=Saturate(None=OneMinus(None=@n18)))
  n20 = Panner(Coordinate=@n12) {speedX=-1.5, speedY=0.7}
  n21 = Saturate(None=ConstantBiasScale(None=TextureSample(UVs=@n20) {texture=&T_VCVFX_SphereMaskTimeOffset_A}.R) {bias=-0.15, scale=1.5})
  n22 = Multiply(A=Subtract(A=@n11, B=Multiply(A=@n14, B=@n19)), B=Add(A=@n11, B=Multiply(A=@n21, B=@n19)))
  n23 = ConstantBiasScale(None=@n22) {bias=-0.1, scale=1024}
  n24 = ConstantBiasScale(None=@n22) {bias=-0.12, scale=1024}
  n25 = Saturate(None=Multiply(A=Saturate(None=@n23), B=OneMinus(None=@n24)))
  n26 = Multiply(A=ScalarParameter() {defaultValue=1, parameterName=Line}, B=@n25) {constA=45}
  n27 = Multiply(A=@n25) {constB=50}
  n28 = ScalarParameter() {defaultValue=1, parameterName=Grid}
  n29 = TextureSample(UVs=Panner(Coordinate=TextureCoordinate()) {speedX=-1.5}) {samplerType=SAMPLERTYPE_LinearColor, texture=/DealershipSharedLibrary/FX/Textures/Gradient_Masks/T_VCVFX_Trail_Heartbeat_TexThread_A}
  n30 = Multiply(A=Multiply(A=@n29.RGB, B=@n29.RGB)) {constB=3}
  n31 = Add(A=Multiply(A=Add(A=@n3, B=@n7)) {constB=3}, B=Add(A=Add(A=@n26, B=@n27), B=Multiply(A=@n28, B=@n30) {constA=15}))
  MaterialFunctionCall(Emissive=Multiply(A=ParticleColor().RGB, B=@n31)) {materialFunction=&MF_FX_EmissiveAdjustments}
  n32 = Sine(None=ComponentMask(None=TextureCoordinate()) {g=true}) {period=1.95}
  Multiply(A=Saturate(None=Multiply(A=@n31)), B=Round(None=Power(Base=@n32) {constExponent=0.9}))
 ]
}
